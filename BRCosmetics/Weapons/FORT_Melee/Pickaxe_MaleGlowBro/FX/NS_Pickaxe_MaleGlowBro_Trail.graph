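# Niagara ribbon trail for the MaleGlowBro pickaxe, built on the FN_AnimTrail modules.
# One emitter spawns two short-lived anim trails, initializes their bone indices and selects width by spawn group.
# Color blends a warm tint with a time-of-day color under a curve-driven alpha fade.
# A separate black DarkTrailColor fades on its own curve, and two ribbon renderers draw the particles.
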NiagaraSystem /BRCosmetics/Weapons/FORT_Melee/Pickaxe_MaleGlowBro/FX/NS_Pickaxe_MaleGlowBro_Trail
dependencies {
 usedDataInterfaces [&NiagaraDataInterfaceCurve, &NiagaraDataInterfaceSocketReader]
 usedDynamicInputs []
 usedModules [
  &self:SystemSpawnScript.SystemScriptSource.SystemScriptGraph.NiagaraNodeAssignment_1.SetVariables_#1
  &SystemState
 ]
}
emitters {FN_AnimTrail {
  emitterSpawnScript [&FN_RibbonIDOffsetUtility()]
  emitterUpdateScript [
   &EmitterState(Life Cycle Mode=::NewEnumerator1, Loop Duration=0.25, UseLoopDelay=false)
   &FN_SpawnAnimTrail(Number of Anim Trails=2)
  ]
  particleSpawnScript [
   &FN_InitializeAnimTrailBoneIndecies()
   &FN_InitializeAnimTrailBoneIndecies(AnimTrail Selection=1) {module=FN_InitializeAnimTrailBoneIndecies001}
   &FN_InitializeAnimTrail(
    AnimTrailWidth=&MergeHelper_FloatSwitch(0=0.9, 1=1, 2=?, 3=?, Index=$Particles.SpawnGroup, Num=::NewEnumerator0)
    Color=<rgba 1 1 1 0.1304>
    Lifetime=0.22
   )
  ]
  particleUpdateScript [
   &SolveForcesAndVelocity(Clamp Velocity=false, Limit Acceleration=false)
   &FN_UpdateAnimTrail() {module=FN_UpdateAnimTrail001}
   &ScaleColor(
    Scale Alpha=&FloatFromCurve(
     FloatCurve=DataInterface() {propertyValues {Curve {keys [{arriveTangent=-18.1546, interpMode=RCIM_Cubic, leaveTangent=-18.1546, tangentMode=RCTM_User, time=0.3, value=1}, <at 0.5 0 RCIM_Cubic>]}}}
     Scale Curve=0.4
    )
    Scale RGB=&MakeVectorFromLinearColorRGB(Color=&LerpLinearColors(
      StartColor=<rgba 0.7865 0.4641 0.2048 1>
      EndColor=&FNTimeOfDayColor(
       Night Color=<rgba 0.1504 0.266 0.4375 1>
       Morning Color=<rgba 1 0.5091 0.5377 1>
       Day Color=<rgba 1 0.7191 0.5573 1>
       Evening Color=<rgba 1 0.5091 0.5377 1>
      )
      LerpFactor=0.25
     ))
    ScaleA=true
    ScaleRGB=true
   )
   /SetVariables(Particles.DarkTrailColor:LinearColor=&MakeLinearColorFromVectorAndFloat(
     "Vector (RGB)"=<0 0 0>
     "Float (Alpha)"=&FloatFromCurve(
      FloatCurve=DataInterface() {propertyValues {Curve {keys [{arriveTangent=-9.6211, interpMode=RCIM_Cubic, leaveTangent=-9.6211, tangentMode=RCTM_User, time=0.2, value=1}, <at 0.5 0 RCIM_Cubic>]}}}
      Scale Curve=0.5
     )
    )) {isSetParameters=true}
  ]
  rendererClasses [&NiagaraRibbonRendererProperties]
  renderers [
   {rendererClass=&NiagaraRibbonRendererProperties}
   {rendererClass=&NiagaraRibbonRendererProperties, rendererIndex=1}
  ]
 }}
summary {
 systemName=NS_Pickaxe_MaleGlowBro_Trail
 emitters [{emitterName=FN_AnimTrail, rendererClasses [&NiagaraRibbonRendererProperties]}]
 userVariables [User.RibbonIDOffset {type=NiagaraInt32}]
}
